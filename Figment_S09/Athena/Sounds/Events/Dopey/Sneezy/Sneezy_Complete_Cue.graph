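# Sound cue for the Sneezy completion sound in the Figment_S09 Dopey event, routed to a map-wide events sound class.
# A root mixer layers close and distant branches, each with its own attenuation asset borrowed from the Snow sounds.
# The close branch mixes a modulated completion wave with an ending track on a fixed delay, while the distant branch plays a separate distant wave.
# Overridden concurrency caps simultaneous instances, stopping the oldest.
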SoundCue /Figment_S09/Athena/Sounds/Events/Dopey/Sneezy/Sneezy_Complete_Cue
root=Mixer_0
nodes [
 {id=Attenuation_0, children [Mixer_1], properties {attenuationSettings=/Figment_S09/Athena/Sounds/Events/Dopey/Snow/Snow_MLTI_Close_Att}}
 {
  id=Attenuation_1
  children [WavePlayer_1]
  properties {attenuationSettings=/Figment_S09/Athena/Sounds/Events/Dopey/Snow/Snow_MLTI_Distant_Att}
 }
 {id=Delay_0, children [WavePlayer_2], properties {delayMax=3, delayMin=3}}
 {id=Mixer_0, children [Attenuation_0, Attenuation_1], properties {inputVolume [1,1]}}
 {id=Mixer_1, children [Modulator_1,Delay_0], properties {inputVolume [1,1]}}
 {id=Modulator_1, children [WavePlayer_0]}
 {id=WavePlayer_0, properties {soundWaveAssetPtr=&/Sneezy_Complete_01}}
 {id=WavePlayer_1, properties {soundWaveAssetPtr=&/Sneezy_Complete_Distant_01}}
 {id=WavePlayer_2, properties {soundWaveAssetPtr=&/SneezyTrack_01_End}}
]
properties {
 bOverrideConcurrency=true
 maxDistance=2097152
 soundClassObject=/Figment_S09/Athena/Sounds/Events/SoundClasses/Fort_Events_Mapwide
 volumeMultiplier=2
 attenuationOverrides {falloffDistance=3000, attenuationShapeExtents {x=0}}
 concurrencyOverrides {maxCount=4, resolutionRule=StopOldest}
}
